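; Spotlight Blueprint for the Festivus event stage that, going by its name, serves as a shadowed stage light and scales itself back on weaker platforms.
; The construction script gets the spotlight component and checks for a mobile platform or a platform name of Switch.
; On those platforms it turns off shadow casting and hides the light component.
; BeginPlay and Tick are empty, so the logic runs only at construction.
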(event EventBeginPlay)

(event EventTick
  (params (DeltaSeconds :type Float%20%28single-precision%29)))

(fn ConstructionScript
  (bind _spotlightcomponent (Variables|Light|GetSpotLightComponent))
  (if (or (Fortnite|IsMobilePlatform) (Utilities|String|EqualExactly(String) (Game|GetPlatformName) "Switch"))
    (Rendering|Components|Light|SetCastShadows _spotlightcomponent false)
    (Rendering|SetVisibility _spotlightcomponent false false)))
